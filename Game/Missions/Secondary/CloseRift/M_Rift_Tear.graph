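# M_Rift_Tear is a translucent unlit material for the rift tear effect in the Close Rift secondary mission.
# Two panning plasma textures add to a ramp whose UV tiling wavers on multiplied sine waves, masked by a powered V gradient.
# That mask blends blue into two bright parameterized inner colors and drives a 1 to 1.5 lerp, likely for refraction.
# SimpleGrassWind is repurposed to supply vertex sway.
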Material /Game/Missions/Secondary/CloseRift/M_Rift_Tear
graph {
 properties {
  blendMode=BLEND_Translucent
  pixelDepthOffsetMode=PDOM_Legacy
  refractionMethod=RM_IndexOfRefraction
  shadingModel=MSM_Unlit
  usedShadingModels=Unlit
 }
 nodes [
  n1 = Time()
  n2 = ConstantBiasScale(None=Multiply(A=Sine(None=@n1) {period=0.32}, B=Sine(None=@n1) {period=0.43}))
  n3 = TextureSample(UVs=LinearInterpolate(A=TextureCoordinate() {vTiling=0.95}, B=TextureCoordinate(), Alpha=Clamp(None=@n2))) {texture=/Game/Effects/Fort_Effects/Textures/Ramps/T_Smoke_Shoot_Out_Ramp}
  n4 = Panner() {speedX=0.1, speedY=1}
  n5 = Panner(Coordinate=TextureCoordinate() {uTiling=1.2, vTiling=1.5}) {speedX=-0.1, speedY=0.6}
  n6 = Add(
   A=Add(
    A=TextureSample(UVs=@n4) {texture=/Game/Effects/Fort_Effects/Textures/T_FX_SpringerPlasma}.R
    B=TextureSample(UVs=@n5) {texture=&T_FX_SpringerPlasma}.R
   )
   B=@n3.R
  )
  n7 = Power(Base=ComponentMask(None=TextureCoordinate()) {g=true})
  LinearInterpolate(
   A=Constant3Vector() {constant=<rgba 0.0696 0.0327 0.63 0>}
   B=LinearInterpolate(
    A=VectorParameter() {parameterName=InnerColor1, defaultValue=<rgba 2.6563 0 3 1>}.RGB
    B=VectorParameter() {parameterName=InnerColor2, defaultValue=<rgba 5 0.625 4.995 1>}.RGB
    Alpha=@n3.R
   ) {constB=0}
   Alpha=Multiply(A=@n6, B=@n7)
  )
  n8 = Add(A=@n6) {constB=0.25}
  LinearInterpolate(Alpha=Multiply(A=Clamp(None=@n8), B=@n7)) {constA=1, constB=1.5}
  MaterialFunctionCall!(WindIntensity=Constant() {r=0.5}, WindWeight=Constant() {r=1}, WindSpeed=Constant() {r=0.2}, AdditionalWPO=Constant()) {materialFunction=/Engine/Functions/Engine_MaterialFunctions01/WorldPositionOffset/SimpleGrassWind}
 ]
}
